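; Creative fishing bobber projectile (FloppingRabbit loot variant) that decides what happens when its capsule overlaps an actor.
; On the authority, it tries tethering and dies on success, otherwise skips ignored pawns, adds flopper-spawner big-bob bonuses or attaches pickups.
; If the scalable-float fishing toggle is on, a DoOnce halts movement, tightens the wire, starts the fishing session and spawns a default flopper spawner.
; Otherwise it hides and kills itself.
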(event EventBeginPlay
  (|CallParentFunction))

(event OnComponentBeginOverlap(CapsuleComponent)
  (params (OverlappedComponent :type Primitive%20Component%20Object%20Reference) (OtherActor :type Actor%20Object%20Reference) (OtherComp :type Primitive%20Component%20Object%20Reference) (OtherBodyIndex :type Integer) (bFromSweep :type Boolean) (SweepResult :type Hit%20Result%20Structure%20%28by%20ref%29))
  (bind _wire (Variables|Default|GetWire))
  ;[1 1 1] Overlap - If we overlap water turn off collsion, tighen the rope, start the idle bob, and check to see if we are in a fishing spawner.
  (Networking|SwitchHasAuthority
    (:Authority
      (bind _output_get (Variables|Default|SetOverlappedActor OtherActor))
      (Development|PrintString (Utilities|String|Append "FISHING ACTOR HIT: " (Utilities|GetDisplayName _output_get)) true true "(R=0.939945,G=1.000000,B=0.000000,A=1.000000)" 1e+01)
      (if (not (Variables|Default|GetbDidHitMeatball))
        (bind _btethered (CallFunction|AttemptTether (Variables|Default|GetOverlappedActor)))
        (if _btethered
          (CallFunction|ForceKill)
          (else
            (bind _isignored (CallFunction|IgnoreTetherablePawn (Variables|Default|GetOverlappedActor)))
            (if (not _isignored)
              (bind _asbga_athena_flopper_spawn_parent (Utilities|Casting|CastToBGA_Athena_FlopperSpawn_Parent (Variables|Default|GetOverlappedActor))
                (:then
                  (Variables|WaterAndCatchingFunctionality|SetChanceForBigBob (+ (Class|BGAAthenaFlopperSpawnParent|GetBonusChanceforBigBob _asbga_athena_flopper_spawn_parent) (Variables|WaterAndCatchingFunctionality|GetChanceForBigBob))))
                (:CastFailed
                  (Utilities|Casting|CastToFortWaterBodyActor (Variables|Default|GetOverlappedActor)
                    (:then
                      _)
                    (:CastFailed
                      (Utilities|Casting|CastToFortAthenaWater (Variables|Default|GetOverlappedActor)
                        (:then
                          _)
                        (:CastFailed
                          (bind _asfort_pickup (Utilities|Casting|CastToFortPickup (Variables|Default|GetOverlappedActor)))
                          (CallFunction|CustomPickupAttach _asfort_pickup))) ; Check Old Water For Creative
                    ))))
              (if (CallFunction|FishingEnabled)
                (Utilities|FlowControl|DoOnce false
                  (:Completed
                    (Variables|WaterAndCatchingFunctionality|SetInWater true)
                    (Transformation|SetWorldRotation (Variables|Default|GetMesh) "0, 0, 0" false false)
                    (Components|Movement|StopMovementImmediately (Variables|Movement|GetProjectileMovementComponent))
                    (Components|Activation|Deactivate (Variables|Default|GetRotatingMovement))
                    (Collision|SetCollisionResponsetoChannel (Variables|Collision|GetCapsuleComponent) "ECC_Pawn" "ECR_Ignore")
                    (Fortnite|Fishing|FireEventBeginFishingSession (Variables|Default|GetPlayerPawn))
                    (Variables|Default|SetbStartedFishingSession true)
                    (CallFunction|CheckforDip)
                    (Utilities|IsValid _wire
                      (:Is%20Valid
                        (Class|BAthenaFloppingRabbitWire|TightenRope _wire))
                      (:Is%20Not%20Valid
                        _))
                    (Ability|SendGameplayEventtoActor (Variables|Default|GetPlayerPawn) (Variables|Tags|GetEvent_InWorld))
                    (CallFunction|HitWater false)
                    (CallFunction|RegisterwithGoalManager)
                    ;[1 1 1] Spawn Default Flopper spawner.
                    (bind _returnvalue (Game|SpawnActorfromClass "/Game/Athena/Items/EnvironmentalItems/FlopperSpawn/BGA_Athena_FlopperSpawn_Default.BGA_Athena_FlopperSpawn_Default_C" (Math|Transform|MakeTransform (Transformation|GetActorLocation)) "AlwaysSpawn" "OverrideRootScale" (Variables|Actor|GetInstigator) false 0 (Variables|Actor|GetInstigator)))
                    (Variables|InSpawnerCheck|SetSpawner _returnvalue)
                    (Variables|InSpawnerCheck|SetSpawnedDefaultSpawner _returnvalue)))
                (else
                  (CallFunction|HideandKill))))))))))

(fn FishingEnabled
  (return (Math|Conversions|ToBoolean(Integer) (Math|Float|Truncate (ScalableFloat|GetValueatLevelwithContext (Variables|Default|GetEnabledRow))))))

(fn ConstructionScript
  (|CallParentFunction))
